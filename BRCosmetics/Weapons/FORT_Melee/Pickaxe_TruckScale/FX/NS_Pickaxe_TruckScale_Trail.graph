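# Niagara swing trail for the TruckScale pickaxe cosmetic, drawn as short-lived anim-trail particles through three ribbon renderers.
# The active Trails_Temp emitter initializes and updates the trail through a system socket reader; the older-style Trails emitter, lacking that reader, is disabled.
# User.Variant lerps between two color palettes, with time-of-day colors in the second.
# System TrailFade and TODColorMultiplier scale the fading colors and alpha.
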NiagaraSystem /BRCosmetics/Weapons/FORT_Melee/Pickaxe_TruckScale/FX/NS_Pickaxe_TruckScale_Trail
dependencies {
 usedDataInterfaces [&NiagaraDataInterfaceColorCurve, &NiagaraDataInterfaceCurve, &NiagaraDataInterfaceSocketReader]
 usedDynamicInputs [&self:SectionIndexFloatSelector_03]
 usedModules [
  &self:SystemSpawnScript.SystemScriptSource.SystemScriptGraph.NiagaraNodeAssignment_6.SetVariables_#1
  &self:SystemSpawnScript.SystemScriptSource.SystemScriptGraph.NiagaraNodeAssignment_4.SetVariables_#2
  &SystemState
  &self:SystemSpawnScript.SystemScriptSource.SystemScriptGraph.NiagaraNodeAssignment_0.SetVariables_#3
  &self:SystemSpawnScript.SystemScriptSource.SystemScriptGraph.NiagaraNodeAssignment_5.SetVariables_#4
 ]
}
emitters {
 Trails {
  bEnabled=false
  emitterSpawnScript [/SetVariables(
    Emitter.Board Color:LinearColor=&MultiplyLinearColor(
     Linear Color=&LerpLinearColors(
      StartColor=<rgba 100 0 60 1>
      EndColor=&FNTimeOfDayColor(
       Night Color=<rgba 100 8 15.6666 1>
       Morning Color=<rgba 100 4 7.3642 1>
       Day Color=<rgba 100 4 7.3642 1>
       Evening Color=<rgba 100 4 7.3642 1>
      )
      LerpFactor=$User.Variant
     )
     Scale Factor=&MakeVector4_FromLinearColor(LinearColor=$System.TODColorMultiplier)
    )
    Emitter.Color A:LinearColor=&MultiplyLinearColor(
     Linear Color=&LerpLinearColors(
      StartColor=<rgba 1 0.4203 0.7681 1>
      EndColor=&FNTimeOfDayColor(
       Night Color=<rgba 1 0.14 0.1973 1>
       Morning Color=<rgba 1 0.12 0.164 1>
       Day Color=<rgba 1 0.12 0.164 1>
       Evening Color=<rgba 1 0.12 0.164 1>
      )
      LerpFactor=$User.Variant
     )
     Scale Factor=&MakeVector4_FromLinearColor(LinearColor=$System.TODColorMultiplier)
    )
    Emitter.Color B:LinearColor=&MultiplyLinearColor(
     Linear Color=&LerpLinearColors(
      StartColor=<rgba 2 0.8406 1.5362 1>
      EndColor=&FNTimeOfDayColor(Night Color=<rgba 2 0.16 0.252 1>, Morning Color=<rgba 2 0.12 0.214 1>, Day Color=<rgba 2 0.12 0.214 1>, Evening Color=<rgba 2 0.12 0.214 1>)
      LerpFactor=$User.Variant
     )
     Scale Factor=&MakeVector4_FromLinearColor(LinearColor=$System.TODColorMultiplier)
    )
    Emitter.EdgeColor:LinearColor=&MultiplyLinearColor(
     Linear Color=&LerpLinearColors(
      StartColor=<rgba 4 0 2 1>
      EndColor=&FNTimeOfDayColor(Night Color=<rgba 4 0.32 0.504 1>, Morning Color=<rgba 4 0.2 0.39 1>, Day Color=<rgba 4 0.2 0.39 1>, Evening Color=<rgba 4 0.2 0.39 1>)
      LerpFactor=$User.Variant
     )
     Scale Factor=&MakeVector4_FromLinearColor(LinearColor=$System.TODColorMultiplier)
    )
    Emitter.Glow:LinearColor=&MultiplyLinearColor(
     Linear Color=&LerpLinearColors(
      StartColor=<rgba 3 0 1.8 1>
      EndColor=&FNTimeOfDayColor(Night Color=<rgba 3 0.3 0.435 1>, Morning Color=<rgba 3 0.24 0.378 1>, Day Color=<rgba 3 0.24 0.378 1>, Evening Color=<rgba 3 0.24 0.378 1>)
      LerpFactor=$User.Variant
     )
     Scale Factor=&MakeVector4_FromLinearColor(LinearColor=$System.TODColorMultiplier)
    )
   ) {isSetParameters=true, module=SetVariables_#5}]
  emitterUpdateScript [
   &EmitterState()
   &FN_SpawnAnimTrail(Spawn Interval=0)
  ]
  particleSpawnScript [
   &FN_InitializeAnimTrailBoneIndecies()
   &FN_InitializeAnimTrail(Lifetime=0.22) {module=FN_InitializeAnimTrail002}
  ]
  particleUpdateScript [
   &FN_UpdateAnimTrail() {module=FN_UpdateAnimTrail001}
   &DynamicMaterialParameters(
    Index 0 Param 1=1
    Index 0 Param 2=1
    Module.Index0Float0Write=true
    Module.Index0Float1Write=true
    Module.Index0Float2Write=true
    Module.Index0Float3Write=true
   )
   &ScaleColor(
    Scale Alpha=&Multiply_Float(
     A=&MakeFloatFromLinearColor(LinearColor=$System.TODColorMultiplier, Channel:ENiagaraLinearColor_Channels=::NewEnumerator3)
     B=&Multiply_Float(
      A=&FloatFromCurve(FloatCurve=DataInterface() {propertyValues {Curve {keys [{interpMode=RCIM_Cubic, tangentMode=RCTM_User, time=0, value=1}, {interpMode=RCIM_Cubic, tangentMode=RCTM_User, time=1, value=0}]}}})
      B=$System.TrailFade
     )
    )
    Scale RGB=&MakeVectorFromLinearColorRGB(Color=&MultiplyLinearColor(
      Linear Color=&LerpLinearColors(
       StartColor=&ColorFromCurve(ColorCurve=DataInterface() {propertyValues {
          AlphaCurve {keys []}
          BlueCurve {keys [<at 0 0.6715 RCIM_Cubic>,<at 1 0.0696 RCIM_Cubic>]}
          GreenCurve {keys [<at 0 0.343 RCIM_Cubic>,<at 1 0.0234 RCIM_Cubic>]}
          RedCurve {keys [<at 0 1 RCIM_Cubic>,<at 1 0.0928 RCIM_Cubic>]}
         }})
       EndColor=&ColorFromCurve(ColorCurve=DataInterface() {propertyValues {
          AlphaCurve {keys []}
          BlueCurve {keys [<at 0 0.145 RCIM_Cubic>,<at 1 0.0134 RCIM_Cubic>]}
          GreenCurve {keys [<at 0 0.1 RCIM_Cubic>,<at 1 0.0093 RCIM_Cubic>]}
          RedCurve {keys [<at 0 1 RCIM_Cubic>,<at 1 0.0928 RCIM_Cubic>]}
         }})
       LerpFactor=$User.Variant
      )
      Scale Factor=&MakeVector4_FromLinearColor(LinearColor=$System.TODColorMultiplier)
     ))
    ScaleA=true
    ScaleRGB=true
   )
   /SetVariables(Particles.Light_BG_Color:LinearColor=&MakeLinearColorFromVectorAndFloat(
     "Vector (RGB)"=&MakeVectorFromLinearColorRGB(Color=&MultiplyLinearColor(
       Linear Color=&LerpLinearColors(
        StartColor=&ColorFromCurve(ColorCurve=DataInterface() {propertyValues {
           AlphaCurve {keys []}
           BlueCurve {keys [<at 0 0.6715 RCIM_Cubic>,<at 1 0.0696 RCIM_Cubic>]}
           GreenCurve {keys [<at 0 0.343 RCIM_Cubic>,<at 1 0.0234 RCIM_Cubic>]}
           RedCurve {keys [<at 0 1 RCIM_Cubic>,<at 1 0.0928 RCIM_Cubic>]}
          }})
        EndColor=&FNTimeOfDayColor(
         Night Color=&ColorFromCurve(ColorCurve=DataInterface() {propertyValues {
            AlphaCurve {keys []}
            BlueCurve {keys [<at 0 0.69 RCIM_Cubic>,<at 1 0.0621 RCIM_Cubic>]}
            GreenCurve {keys [<at 0 0.38 RCIM_Cubic>,<at 1 0.035 RCIM_Cubic>]}
            RedCurve {keys [<at 0 1 RCIM_Cubic>,<at 1 0.1 RCIM_Cubic>]}
           }})
         Morning Color=&ColorFromCurve(ColorCurve=DataInterface() {propertyValues {
            AlphaCurve {keys []}
            BlueCurve {keys [<at 0 0.24 RCIM_Cubic>,<at 1 0.0267 RCIM_Cubic>]}
            GreenCurve {keys [<at 0 0.2 RCIM_Cubic>,<at 1 0.0232 RCIM_Cubic>]}
            RedCurve {keys [<at 0 1 RCIM_Cubic>,<at 1 0.0928 RCIM_Cubic>]}
           }})
         Day Color=&ColorFromCurve(ColorCurve=DataInterface() {propertyValues {
            AlphaCurve {keys []}
            BlueCurve {keys [<at 0 0.24 RCIM_Cubic>,<at 1 0.0267 RCIM_Cubic>]}
            GreenCurve {keys [<at 0 0.2 RCIM_Cubic>,<at 1 0.0232 RCIM_Cubic>]}
            RedCurve {keys [<at 0 1 RCIM_Cubic>,<at 1 0.0928 RCIM_Cubic>]}
           }})
         Evening Color=&ColorFromCurve(ColorCurve=DataInterface() {propertyValues {
            AlphaCurve {keys []}
            BlueCurve {keys [<at 0 0.24 RCIM_Cubic>,<at 1 0.0267 RCIM_Cubic>]}
            GreenCurve {keys [<at 0 0.2 RCIM_Cubic>,<at 1 0.0232 RCIM_Cubic>]}
            RedCurve {keys [<at 0 1 RCIM_Cubic>,<at 1 0.0928 RCIM_Cubic>]}
           }})
        )
        LerpFactor=$User.Variant
       )
       Scale Factor=&MakeVector4_FromLinearColor(LinearColor=$System.TODColorMultiplier)
      ))
     "Float (Alpha)"=&Multiply_Float(
      A=&MakeFloatFromLinearColor(LinearColor=$System.TODColorMultiplier, Channel:ENiagaraLinearColor_Channels=::NewEnumerator3)
      B=&Multiply_Float(
       A=&FloatFromCurve(FloatCurve=DataInterface() {propertyValues {Curve {keys [{interpMode=RCIM_Cubic, tangentMode=RCTM_User, time=0, value=1}, {interpMode=RCIM_Cubic, tangentMode=RCTM_User, time=1, value=0}]}}})
       B=$System.TrailFade
      )
     )
    )) {isSetParameters=true, module=SetVariables_#6}
   /SetVariables(Particles.Dark_BG_Color:LinearColor=&MakeLinearColorFromVectorAndFloat(
     "Vector (RGB)"=&MakeVectorFromLinearColorRGB(Color=&MultiplyLinearColor(
       Linear Color=&LerpLinearColors(
        StartColor=&ColorFromCurve(ColorCurve=DataInterface() {propertyValues {
           AlphaCurve {keys []}
           BlueCurve {keys [<at 0 0.1119 RCIM_Cubic>,<at 1 0.0289 RCIM_Cubic>]}
           GreenCurve {keys [<at 0 0.0249 RCIM_Cubic>,<at 1 0.0064 RCIM_Cubic>]}
           RedCurve {keys [<at 0 0.1681 RCIM_Cubic>,<at 1 0.0434 RCIM_Cubic>]}
          }})
        EndColor=&FNTimeOfDayColor(
         Night Color=&ColorFromCurve(ColorCurve=DataInterface() {propertyValues {
            AlphaCurve {keys []}
            BlueCurve {keys [<at 0 0.3333 RCIM_Cubic>,<at 1 0.109 RCIM_Cubic>]}
            GreenCurve {keys [<at 0 0.1 RCIM_Cubic>,<at 1 0.044 RCIM_Cubic>]}
            RedCurve {keys [<at 0 0.5 RCIM_Cubic>,<at 1 0.2 RCIM_Cubic>]}
           }})
         Morning Color=&ColorFromCurve(ColorCurve=DataInterface() {propertyValues {
            AlphaCurve {keys []}
            BlueCurve {keys [<at 0 0.0324 RCIM_Cubic>,<at 1 0.0125 RCIM_Cubic>]}
            GreenCurve {keys [<at 0 0.0252 RCIM_Cubic>,<at 1 0.0109 RCIM_Cubic>]}
            RedCurve {keys [<at 0 0.1681 RCIM_Cubic>,<at 1 0.0434 RCIM_Cubic>]}
           }})
         Day Color=&ColorFromCurve(ColorCurve=DataInterface() {propertyValues {
            AlphaCurve {keys []}
            BlueCurve {keys [<at 0 0.0324 RCIM_Cubic>,<at 1 0.0125 RCIM_Cubic>]}
            GreenCurve {keys [<at 0 0.0252 RCIM_Cubic>,<at 1 0.0109 RCIM_Cubic>]}
            RedCurve {keys [<at 0 0.1681 RCIM_Cubic>,<at 1 0.0434 RCIM_Cubic>]}
           }})
         Evening Color=&ColorFromCurve(ColorCurve=DataInterface() {propertyValues {
            AlphaCurve {keys []}
            BlueCurve {keys [<at 0 0.0324 RCIM_Cubic>,<at 1 0.0125 RCIM_Cubic>]}
            GreenCurve {keys [<at 0 0.0252 RCIM_Cubic>,<at 1 0.0109 RCIM_Cubic>]}
            RedCurve {keys [<at 0 0.1681 RCIM_Cubic>,<at 1 0.0434 RCIM_Cubic>]}
           }})
        )
        LerpFactor=$User.Variant
       )
       Scale Factor=&MakeVector4_FromLinearColor(LinearColor=$System.TODColorMultiplier)
      ))
     "Float (Alpha)"=&Multiply_Float(
      A=&MakeFloatFromLinearColor(LinearColor=$System.TODColorMultiplier, Channel:ENiagaraLinearColor_Channels=::NewEnumerator3)
      B=&Multiply_Float(
       A=&FloatFromCurve(
        FloatCurve=DataInterface() {propertyValues {Curve {keys [{interpMode=RCIM_Cubic, tangentMode=RCTM_User, time=0, value=1}, {interpMode=RCIM_Cubic, tangentMode=RCTM_User, time=1, value=0}]}}}
        Scale Curve=2
       )
       B=$System.TrailFade
      )
     )
    )) {isSetParameters=true, module=SetVariables_#7}
  ]
  rendererClasses [&NiagaraRibbonRendererProperties]
  renderers [
   {rendererClass=&NiagaraRibbonRendererProperties}
   {rendererClass=&NiagaraRibbonRendererProperties, rendererIndex=1}
   {rendererClass=&NiagaraRibbonRendererProperties, rendererIndex=2}
  ]
 }
 Trails_Temp {
  emitterSpawnScript [/SetVariables(
    Emitter.Board Color:LinearColor=&MultiplyLinearColor(
     Linear Color=&LerpLinearColors(
      StartColor=<rgba 100 0 60 1>
      EndColor=&FNTimeOfDayColor(
       Night Color=<rgba 100 8 15.6666 1>
       Morning Color=<rgba 100 4 7.3642 1>
       Day Color=<rgba 100 4 7.3642 1>
       Evening Color=<rgba 100 4 7.3642 1>
      )
      LerpFactor=$User.Variant
     )
     Scale Factor=&MakeVector4_FromLinearColor(LinearColor=$System.TODColorMultiplier)
    )
    Emitter.Color A:LinearColor=&MultiplyLinearColor(
     Linear Color=&LerpLinearColors(
      StartColor=<rgba 1 0.4203 0.7681 1>
      EndColor=&FNTimeOfDayColor(
       Night Color=<rgba 1 0.14 0.1973 1>
       Morning Color=<rgba 1 0.12 0.164 1>
       Day Color=<rgba 1 0.12 0.164 1>
       Evening Color=<rgba 1 0.12 0.164 1>
      )
      LerpFactor=$User.Variant
     )
     Scale Factor=&MakeVector4_FromLinearColor(LinearColor=$System.TODColorMultiplier)
    )
    Emitter.Color B:LinearColor=&MultiplyLinearColor(
     Linear Color=&LerpLinearColors(
      StartColor=<rgba 2 0.8406 1.5362 1>
      EndColor=&FNTimeOfDayColor(Night Color=<rgba 2 0.16 0.252 1>, Morning Color=<rgba 2 0.12 0.214 1>, Day Color=<rgba 2 0.12 0.214 1>, Evening Color=<rgba 2 0.12 0.214 1>)
      LerpFactor=$User.Variant
     )
     Scale Factor=&MakeVector4_FromLinearColor(LinearColor=$System.TODColorMultiplier)
    )
    Emitter.EdgeColor:LinearColor=&MultiplyLinearColor(
     Linear Color=&LerpLinearColors(
      StartColor=<rgba 4 0 2 1>
      EndColor=&FNTimeOfDayColor(Night Color=<rgba 4 0.32 0.504 1>, Morning Color=<rgba 4 0.2 0.39 1>, Day Color=<rgba 4 0.2 0.39 1>, Evening Color=<rgba 4 0.2 0.39 1>)
      LerpFactor=$User.Variant
     )
     Scale Factor=&MakeVector4_FromLinearColor(LinearColor=$System.TODColorMultiplier)
    )
    Emitter.Glow:LinearColor=&MultiplyLinearColor(
     Linear Color=&LerpLinearColors(
      StartColor=<rgba 3 0 1.8 1>
      EndColor=&FNTimeOfDayColor(Night Color=<rgba 3 0.3 0.435 1>, Morning Color=<rgba 3 0.24 0.378 1>, Day Color=<rgba 3 0.24 0.378 1>, Evening Color=<rgba 3 0.24 0.378 1>)
      LerpFactor=$User.Variant
     )
     Scale Factor=&MakeVector4_FromLinearColor(LinearColor=$System.TODColorMultiplier)
    )
   ) {isSetParameters=true, module=SetVariables_#5}]
  emitterUpdateScript [
   &EmitterState()
   &FN_SpawnAnimTrail(Spawn Interval=0)
  ]
  particleSpawnScript [
   &FN_InitializeAnimTrailBoneIndecies()
   /Game/Effects/Niagara/Modules/AnimTrail/FN_InitializeAnimTrail/FN_InitializeAnimTrail_SocketReader/FN_InitializeAnimTrail_SocketReader(Lifetime=0.22, Socket Reader:NiagaraDataInterfaceSocketReader=$System.Socket Reader)
  ]
  particleUpdateScript [
   /Game/Effects/Niagara/Modules/AnimTrail/FN_UpdateAnimTrail/FN_UpdateAnimTrail_SocketReader/FN_UpdateAnimTrail_SocketReader(Socket Reader:NiagaraDataInterfaceSocketReader=$System.Socket Reader)
   &DynamicMaterialParameters(
    Index 0 Param 1=1
    Index 0 Param 2=1
    Module.Index0Float0Write=true
    Module.Index0Float1Write=true
    Module.Index0Float2Write=true
    Module.Index0Float3Write=true
   )
   &ScaleColor(
    Scale Alpha=&Multiply_Float(
     A=&MakeFloatFromLinearColor(LinearColor=$System.TODColorMultiplier, Channel:ENiagaraLinearColor_Channels=::NewEnumerator3)
     B=&Multiply_Float(
      A=&FloatFromCurve(FloatCurve=DataInterface() {propertyValues {Curve {keys [{interpMode=RCIM_Cubic, tangentMode=RCTM_User, time=0, value=1}, {interpMode=RCIM_Cubic, tangentMode=RCTM_User, time=1, value=0}]}}})
      B=$System.TrailFade
     )
    )
    Scale RGB=&MakeVectorFromLinearColorRGB(Color=&MultiplyLinearColor(
      Linear Color=&LerpLinearColors(
       StartColor=&ColorFromCurve(ColorCurve=DataInterface() {propertyValues {
          AlphaCurve {keys []}
          BlueCurve {keys [<at 0 0.6715 RCIM_Cubic>,<at 1 0.0696 RCIM_Cubic>]}
          GreenCurve {keys [<at 0 0.343 RCIM_Cubic>,<at 1 0.0234 RCIM_Cubic>]}
          RedCurve {keys [<at 0 1 RCIM_Cubic>,<at 1 0.0928 RCIM_Cubic>]}
         }})
       EndColor=&ColorFromCurve(ColorCurve=DataInterface() {propertyValues {
          AlphaCurve {keys []}
          BlueCurve {keys [<at 0 0.145 RCIM_Cubic>,<at 1 0.0134 RCIM_Cubic>]}
          GreenCurve {keys [<at 0 0.1 RCIM_Cubic>,<at 1 0.0093 RCIM_Cubic>]}
          RedCurve {keys [<at 0 1 RCIM_Cubic>,<at 1 0.0928 RCIM_Cubic>]}
         }})
       LerpFactor=$User.Variant
      )
      Scale Factor=&MakeVector4_FromLinearColor(LinearColor=$System.TODColorMultiplier)
     ))
    ScaleA=true
    ScaleRGB=true
   )
   /SetVariables(Particles.Light_BG_Color:LinearColor=&MakeLinearColorFromVectorAndFloat(
     "Vector (RGB)"=&MakeVectorFromLinearColorRGB(Color=&MultiplyLinearColor(
       Linear Color=&LerpLinearColors(
        StartColor=&ColorFromCurve(ColorCurve=DataInterface() {propertyValues {
           AlphaCurve {keys []}
           BlueCurve {keys [<at 0 0.6715 RCIM_Cubic>,<at 1 0.0696 RCIM_Cubic>]}
           GreenCurve {keys [<at 0 0.343 RCIM_Cubic>,<at 1 0.0234 RCIM_Cubic>]}
           RedCurve {keys [<at 0 1 RCIM_Cubic>,<at 1 0.0928 RCIM_Cubic>]}
          }})
        EndColor=&FNTimeOfDayColor(
         Night Color=&ColorFromCurve(ColorCurve=DataInterface() {propertyValues {
            AlphaCurve {keys []}
            BlueCurve {keys [<at 0 0.69 RCIM_Cubic>,<at 1 0.0621 RCIM_Cubic>]}
            GreenCurve {keys [<at 0 0.38 RCIM_Cubic>,<at 1 0.035 RCIM_Cubic>]}
            RedCurve {keys [<at 0 1 RCIM_Cubic>,<at 1 0.1 RCIM_Cubic>]}
           }})
         Morning Color=&ColorFromCurve(ColorCurve=DataInterface() {propertyValues {
            AlphaCurve {keys []}
            BlueCurve {keys [<at 0 0.24 RCIM_Cubic>,<at 1 0.0267 RCIM_Cubic>]}
            GreenCurve {keys [<at 0 0.2 RCIM_Cubic>,<at 1 0.0232 RCIM_Cubic>]}
            RedCurve {keys [<at 0 1 RCIM_Cubic>,<at 1 0.0928 RCIM_Cubic>]}
           }})
         Day Color=&ColorFromCurve(ColorCurve=DataInterface() {propertyValues {
            AlphaCurve {keys []}
            BlueCurve {keys [<at 0 0.24 RCIM_Cubic>,<at 1 0.0267 RCIM_Cubic>]}
            GreenCurve {keys [<at 0 0.2 RCIM_Cubic>,<at 1 0.0232 RCIM_Cubic>]}
            RedCurve {keys [<at 0 1 RCIM_Cubic>,<at 1 0.0928 RCIM_Cubic>]}
           }})
         Evening Color=&ColorFromCurve(ColorCurve=DataInterface() {propertyValues {
            AlphaCurve {keys []}
            BlueCurve {keys [<at 0 0.24 RCIM_Cubic>,<at 1 0.0267 RCIM_Cubic>]}
            GreenCurve {keys [<at 0 0.2 RCIM_Cubic>,<at 1 0.0232 RCIM_Cubic>]}
            RedCurve {keys [<at 0 1 RCIM_Cubic>,<at 1 0.0928 RCIM_Cubic>]}
           }})
        )
        LerpFactor=$User.Variant
       )
       Scale Factor=&MakeVector4_FromLinearColor(LinearColor=$System.TODColorMultiplier)
      ))
     "Float (Alpha)"=&Multiply_Float(
      A=&MakeFloatFromLinearColor(LinearColor=$System.TODColorMultiplier, Channel:ENiagaraLinearColor_Channels=::NewEnumerator3)
      B=&Multiply_Float(
       A=&FloatFromCurve(FloatCurve=DataInterface() {propertyValues {Curve {keys [{interpMode=RCIM_Cubic, tangentMode=RCTM_User, time=0, value=1}, {interpMode=RCIM_Cubic, tangentMode=RCTM_User, time=1, value=0}]}}})
       B=$System.TrailFade
      )
     )
    )) {isSetParameters=true, module=SetVariables_#6}
   /SetVariables(Particles.Dark_BG_Color:LinearColor=&MakeLinearColorFromVectorAndFloat(
     "Vector (RGB)"=&MakeVectorFromLinearColorRGB(Color=&MultiplyLinearColor(
       Linear Color=&LerpLinearColors(
        StartColor=&ColorFromCurve(ColorCurve=DataInterface() {propertyValues {
           AlphaCurve {keys []}
           BlueCurve {keys [<at 0 0.1119 RCIM_Cubic>,<at 1 0.0289 RCIM_Cubic>]}
           GreenCurve {keys [<at 0 0.0249 RCIM_Cubic>,<at 1 0.0064 RCIM_Cubic>]}
           RedCurve {keys [<at 0 0.1681 RCIM_Cubic>,<at 1 0.0434 RCIM_Cubic>]}
          }})
        EndColor=&FNTimeOfDayColor(
         Night Color=&ColorFromCurve(ColorCurve=DataInterface() {propertyValues {
            AlphaCurve {keys []}
            BlueCurve {keys [<at 0 0.3333 RCIM_Cubic>,<at 1 0.109 RCIM_Cubic>]}
            GreenCurve {keys [<at 0 0.1 RCIM_Cubic>,<at 1 0.044 RCIM_Cubic>]}
            RedCurve {keys [<at 0 0.5 RCIM_Cubic>,<at 1 0.2 RCIM_Cubic>]}
           }})
         Morning Color=&ColorFromCurve(ColorCurve=DataInterface() {propertyValues {
            AlphaCurve {keys []}
            BlueCurve {keys [<at 0 0.0324 RCIM_Cubic>,<at 1 0.0125 RCIM_Cubic>]}
            GreenCurve {keys [<at 0 0.0252 RCIM_Cubic>,<at 1 0.0109 RCIM_Cubic>]}
            RedCurve {keys [<at 0 0.1681 RCIM_Cubic>,<at 1 0.0434 RCIM_Cubic>]}
           }})
         Day Color=&ColorFromCurve(ColorCurve=DataInterface() {propertyValues {
            AlphaCurve {keys []}
            BlueCurve {keys [<at 0 0.0324 RCIM_Cubic>,<at 1 0.0125 RCIM_Cubic>]}
            GreenCurve {keys [<at 0 0.0252 RCIM_Cubic>,<at 1 0.0109 RCIM_Cubic>]}
            RedCurve {keys [<at 0 0.1681 RCIM_Cubic>,<at 1 0.0434 RCIM_Cubic>]}
           }})
         Evening Color=&ColorFromCurve(ColorCurve=DataInterface() {propertyValues {
            AlphaCurve {keys []}
            BlueCurve {keys [<at 0 0.0324 RCIM_Cubic>,<at 1 0.0125 RCIM_Cubic>]}
            GreenCurve {keys [<at 0 0.0252 RCIM_Cubic>,<at 1 0.0109 RCIM_Cubic>]}
            RedCurve {keys [<at 0 0.1681 RCIM_Cubic>,<at 1 0.0434 RCIM_Cubic>]}
           }})
        )
        LerpFactor=$User.Variant
       )
       Scale Factor=&MakeVector4_FromLinearColor(LinearColor=$System.TODColorMultiplier)
      ))
     "Float (Alpha)"=&Multiply_Float(
      A=&MakeFloatFromLinearColor(LinearColor=$System.TODColorMultiplier, Channel:ENiagaraLinearColor_Channels=::NewEnumerator3)
      B=&Multiply_Float(
       A=&FloatFromCurve(
        FloatCurve=DataInterface() {propertyValues {Curve {keys [{interpMode=RCIM_Cubic, tangentMode=RCTM_User, time=0, value=1}, {interpMode=RCIM_Cubic, tangentMode=RCTM_User, time=1, value=0}]}}}
        Scale Curve=2
       )
       B=$System.TrailFade
      )
     )
    )) {isSetParameters=true, module=SetVariables_#7}
  ]
  rendererClasses [&NiagaraRibbonRendererProperties]
  renderers [
   {rendererClass=&NiagaraRibbonRendererProperties}
   {rendererClass=&NiagaraRibbonRendererProperties, rendererIndex=1}
   {rendererClass=&NiagaraRibbonRendererProperties, rendererIndex=2}
  ]
 }
}
summary {
 systemName=NS_Pickaxe_TruckScale_Trail
 emitters [
  {bEnabled=false, emitterName=Trails, rendererClasses [&NiagaraRibbonRendererProperties]}
  {emitterName=Trails_Temp, rendererClasses [&NiagaraRibbonRendererProperties]}
 ]
 userVariables [
  User.bIsFrontEnd {type=NiagaraBool}
  User.bIsFrontEndPreview {type=NiagaraBool}
  User.RibbonIDOffset {type=NiagaraInt32}
  User.SwingSectionIndex {type=NiagaraInt32}
  User.Variant {type=NiagaraFloat, defaultValue=0, description=""}
 ]
}
